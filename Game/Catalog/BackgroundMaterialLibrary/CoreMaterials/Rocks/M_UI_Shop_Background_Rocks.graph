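# Translucent UI material for an item shop tile background of rocks, tinted by three colors from the shared background-parameter function, with a rounded-corner function.
# One packed rocks texture supplies lines, rock shadows and watercolor, drifting through a time-driven rotation center.
# Offset resamples of the line channel are subtracted to derive glow and shadow edges.
# Radial and bottom gradients, flow-mapped clouds and noise provide further layers, composited with lerps and a final color dodge.
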Material /Game/Catalog/BackgroundMaterialLibrary/CoreMaterials/Rocks/M_UI_Shop_Background_Rocks
graph {
 properties {bAutomaticallySetUsageInEditor=false, blendMode=BLEND_Translucent, materialDomain=MD_UI, usedShadingModels=Default Lit}
 nodes [
  Constant3Vector() {constant=<rgba 0.0844 0.3763 0.9823 1>}
  Constant3Vector() {constant=<rgba 0.0423 0.0802 0.1946 1>}
  Constant3Vector() {constant=<rgba 0.0086 0 0.0242 1>}
  MaterialFunctionCall!(Input Alpha=Constant() {r=1}, bOverrideCornerRadius, Corner Radius Override) {materialFunction=/FortUILibrary/Systems/ItemShopTile/Materials/MaterialFunctions/MF_UI_ItemShop_RoundedCorners}
  n1 = MaterialFunctionCall() {materialFunction=/FortUILibrary/Systems/ItemShopTile/Materials/MaterialFunctions/MF_UI_ItemShop_BackgroundParams}
  NamedRerouteDeclaration(None=@n1.Color 1) {name=Primary Color, nodeColor=<rgba 0 0 0 1>}
  n2 = TextureCoordinate()
  n3 = ScalarParameter() {defaultValue=3, group=Radial Gradient, parameterName=Radial Gradient Scale}
  n4 = ScalarParameter() {group=Radial Gradient, parameterName=Radial Gradient Offset X}
  n5 = ScalarParameter() {defaultValue=0.5, group=Radial Gradient, parameterName=Radial Gradient Offset Y}
  n6 = MaterialFunctionCall(UV=Add(
    A=MaterialFunctionCall(UVs=@n2, Texture Scale=@n3) {materialFunction=&ScaleUVsByCenter}.UVs
    B=AppendVector(A=@n4, B=@n5)
   ) {constB=0.5}) {materialFunction=&MF_UI_UVScaling}
  NamedRerouteDeclaration(None=Saturate(None=MaterialFunctionCall(UV=@n6.Scaled To Fill) {materialFunction=&MF_UI_RadialGradient}.Smooth)) {name=Radial Gradient Mask, nodeColor=<rgba 0 0 0 1>}
  n7 = NamedRerouteUsage()
  n8 = Saturate(None=MaterialFunctionCall(Input=ComponentMask(None=@n7) {g=true}, Input Low=Constant(), Input High=Constant() {r=0.7}) {materialFunction=&RemapValue_0-1}.Result)
  NamedRerouteDeclaration(None=LinearInterpolate(A=NamedRerouteUsage(), B=NamedRerouteUsage(), Alpha=SmoothStep(Value=@n8))) {name=Background Gradient, nodeColor=<rgba 0 0 0 1>}
  n9 = TextureSample(UVs=NamedRerouteUsage()) {samplerType=SAMPLERTYPE_LinearColor, texture=/Game/Catalog/BackgroundMaterialLibrary/Textures/T_UI_Rocks}
  n10 = LinearInterpolate(Alpha=Multiply(A=@n9.B, B=@n9.B)) {constA=0.1}
  NamedRerouteDeclaration(None=Saturate(None=Multiply(A=@n10) {constB=2})) {name=Watercolor Texture, nodeColor=<rgba 0 0 0 1>}
  n11 = MaterialFunctionCall(UV=NamedRerouteDeclaration(None=MaterialFunctionCall() {materialFunction=&GetUserInterfaceUV}.Normalized UV) {name=Normalized UV, nodeColor=<rgba 0 0 0 1>}) {materialFunction=&MF_UI_UVScaling}
  NamedRerouteDeclaration(None=@n11.Scaled to Fit) {name=Scale to Fit, nodeColor=<rgba 0 0 0 1>}
  n12 = Time()
  n13 = MaterialFunctionCall(
   UVs=NamedRerouteUsage()
   Rotation Center=AppendVector(A=Multiply(A=@n12) {constB=-0.004}, B=Multiply(A=@n12) {constB=0.004})
   "Rotation Angle (0-1)"=Constant() {r=0.03}
  ) {materialFunction=&CustomRotator}
  n14 = ScalarParameter() {group=Lines Texture, parameterName=Texture Position X}
  n15 = ScalarParameter() {defaultValue=0.01, group=Lines Texture, parameterName=Texture Position Y}
  n16 = AppendVector(
   A=ComponentMask(None=Add(A=@n13.Rotated Values, B=@n14) {constB=0.2}) {r=true}
   B=ComponentMask(None=Add(A=@n13.Rotated Values, B=@n15) {constB=0.3}) {g=true}
  )
  n17 = TextureSample(UVs=MaterialFunctionCall(UVs=@n16, Texture Scale=ScalarParameter() {defaultValue=0.18, group=Lines Texture, parameterName=Texture Size}) {materialFunction=&ScaleUVsByCenter}.UVs) {samplerType=SAMPLERTYPE_LinearColor, texture=&T_UI_Rocks}
  NamedRerouteDeclaration(None=Saturate(None=@n17.R)) {name=Lines Color, nodeColor=<rgba 0 0 0 1>}
  NamedRerouteDeclaration(None=@n1.Color 2) {name=Secondary Color, nodeColor=<rgba 0 0 0 1>}
  n18 = ScalarParameter() {group=04 - Foreground Clouds, parameterName=Foreground Clouds Min}
  n19 = ScalarParameter() {defaultValue=1, group=04 - Foreground Clouds, parameterName=Foreground Clouds Max}
  n20 = NamedRerouteUsage()
  n21 = Multiply(A=@n20, B=ScalarParameter() {defaultValue=2, group=04 - Foreground Clouds, parameterName=Foreground Clouds Scale}) {constB=0.65}
  n22 = TextureObjectParameter(UVs=@n20) {
   group=04 - Foreground Clouds
   parameterName=Foreground Cloud Texture
   samplerType=SAMPLERTYPE_Alpha
   texture=/Game/UI/Frontend/ItemShop/Art/LastVoice/T_UI_CloudLV
  }
  n23 = Panner(
   Coordinate=@n21
   Time=Constant() {r=1}
   Speed=AppendVector(
    A=ScalarParameter() {defaultValue=0.4, group=04 - Foreground Clouds, parameterName=Foreground Cloud Speed X}
    B=ScalarParameter() {defaultValue=0.4, group=04 - Foreground Clouds, parameterName=Foreground Cloud Speed Y}
   )
  ) {speedX=0.01, speedY=0.01}
  n24 = TextureObjectParameter() {
   group=04 - Foreground Clouds
   parameterName=Foreground Cloud Flowmap
   samplerType=SAMPLERTYPE_Grayscale
   texture=/FortUILibrary/MaterialFunctions/ExampleContent/T_Cloud_R
  }
  n25 = ComponentMask(None=MaterialFunctionCall!(
    UVs=Panner(Coordinate=@n21) {speedX=0.03, speedY=0.015}
    Texture Object=@n22
    Grayscale Flow Map=TextureSample(UVs=@n23, Tex=@n24).RGB
    Flow Direction=ScalarParameter() {defaultValue=0.5, group=04 - Foreground Clouds, parameterName=Foreground Clouds Flow Direction}
    Flow Intensity=ScalarParameter() {defaultValue=0.1, group=04 - Foreground Clouds, parameterName=Foreground Clouds Flow Intensity}
    Flow Speed=ScalarParameter() {defaultValue=0.25, group=04 - Foreground Clouds, parameterName=Foreground Clouds Flow Speed}
   ) {materialFunction=/FortUILibrary/MaterialFunctions/Effects/MF_UI_GrayscaleFlowMap_New}.Distorted Texture) {r=true}
  NamedRerouteDeclaration(None=Saturate(None=SmoothStep(Min=@n18, Max=@n19, Value=@n25))) {name=Foreground Clouds Grayscale, nodeColor=<rgba 0 0 0 1>}
  n26 = Multiply(
   A=MaterialFunctionCall!(UV=NamedRerouteUsage(), Seed) {materialFunction=&MF_UI_Random}.Result
   B=ScalarParameter() {defaultValue=0.3, group=Noise, parameterName=Noise Opacity}
  ) {constB=0.65}
  n27 = ComponentMask(None=NamedRerouteUsage()) {g=true}
  n28 = LinearInterpolate(Alpha=Power(Base=OneMinus(None=@n27))) {constA=0.1}
  NamedRerouteDeclaration(None=Saturate(None=Multiply(A=@n26, B=@n28))) {name=Noise Overlay, nodeColor=<rgba 0 0 0 1>}
  n29 = MaterialFunctionCall!(ScreenUV, TextureSize, ScalePivotPoint) {materialFunction=/FortUILibrary/MaterialFunctions/Basics/MF_UI_ScreenSpaceUVScaling}
  NamedRerouteDeclaration(None=@n29.ScaleY) {name="Scale Y UV  Screen", nodeColor=<rgba 0 0 0 1>}
  NamedRerouteDeclaration(None=@n29.Scaled to Fill) {name="Scale to Fill  Screen", nodeColor=<rgba 0 0 0 1>}
  NamedRerouteDeclaration(None=@n29.Scaled to Fit) {name=Scale to Fit Screen, nodeColor=<rgba 0 0 0 1>}
  NamedRerouteDeclaration(None=@n11.ScaleY) {name=Scale Y UV, nodeColor=<rgba 0 0 0 1>}
  n30 = AppendVector(A=ComponentMask(None=Add(A=@n16) {constB=0.0004}) {r=true}, B=ComponentMask(None=Add(A=@n16) {constB=-0.0004}) {g=true})
  n31 = ScalarParameter() {defaultValue=0.18, group=Lines Texture, parameterName=Texture Size}
  n32 = Subtract(
   A=TextureSample(UVs=MaterialFunctionCall(UVs=@n30, Texture Scale=@n31) {materialFunction=&ScaleUVsByCenter}.UVs) {samplerType=SAMPLERTYPE_LinearColor, texture=&T_UI_Rocks}.R
   B=@n17.R
  )
  NamedRerouteDeclaration(None=Saturate(None=Multiply(A=@n32) {constB=4})) {name=Lines Glow, nodeColor=<rgba 0 0 0 1>}
  n33 = ComponentMask(None=Add(A=@n16) {constB=-0.001}) {r=true}
  n34 = ComponentMask(None=Add(A=@n16) {constB=0.001}) {g=true}
  n35 = TextureSample(UVs=MaterialFunctionCall(UVs=AppendVector(A=@n33, B=@n34), Texture Scale=ScalarParameter() {defaultValue=0.18, group=Lines Texture, parameterName=Texture Size}) {materialFunction=&ScaleUVsByCenter}.UVs) {samplerType=SAMPLERTYPE_LinearColor, texture=&T_UI_Rocks}
  NamedRerouteDeclaration(None=Saturate(None=Subtract(A=@n35.R, B=@n17.R))) {name=Lines Shadow, nodeColor=<rgba 0 0 0 1>}
  NamedRerouteDeclaration(None=Saturate(None=@n17.G)) {name=Rocks Shadow, nodeColor=<rgba 0 0 0 1>}
  n36 = MaterialFunctionCall(Input=ComponentMask(None=NamedRerouteUsage()) {g=true}, Input Low=Constant(), Input High=Constant() {r=1}) {materialFunction=&RemapValue_0-1}
  NamedRerouteDeclaration(None=SmoothStep(Value=Saturate(None=@n36.Result)) {constMax=1.4, constMin=0.7}) {name=Bottom Gradient, nodeColor=<rgba 0 0 0 1>}
  NamedRerouteDeclaration(None=@n11.ScaleX) {name=Scale X UV, nodeColor=<rgba 0 0 0 1>}
  NamedRerouteDeclaration(None=@n11.Scaled To Fill) {name=Scale to Fill, nodeColor=<rgba 0 0 0 1>}
  NamedRerouteDeclaration(None=OneMinus(None=@n8)) {name=Gradient Mask, nodeColor=<rgba 0 0 0 1>}
  NamedRerouteDeclaration(None=@n1.Color 3) {name=Accent Color, nodeColor=<rgba 0 0 0 1>}
  NamedRerouteUsage()
  n37 = NamedRerouteUsage()
  n38 = NamedRerouteUsage()
  n39 = NamedRerouteUsage()
  n40 = NamedRerouteUsage()
  n41 = LinearInterpolate(
   A=LinearInterpolate(A=LinearInterpolate(A=@n37, B=@n38, Alpha=@n39), B=NamedRerouteUsage(), Alpha=Multiply(A=@n40))
   B=NamedRerouteUsage()
   Alpha=Multiply(A=NamedRerouteUsage()) {constB=0.6}
  ) {constA=1, constB=0}
  n42 = Multiply(A=LinearInterpolate(B=NamedRerouteUsage(), Alpha=NamedRerouteUsage())) {constB=0.1}
  n43 = NamedRerouteUsage()
  n44 = NamedRerouteUsage()
  n45 = LinearInterpolate(
   A=LinearInterpolate(A=Add(A=@n41, B=@n42) {constB=0}, B=Constant(), Alpha=NamedRerouteUsage()) {constB=0}
   B=NamedRerouteUsage()
   Alpha=Multiply(A=LinearInterpolate(B=@n43, Alpha=@n44)) {constB=0.4}
  )
  n46 = NamedRerouteUsage()
  n47 = NamedRerouteUsage()
  Saturate(None=MaterialFunctionCall!(Blend=NamedRerouteUsage(), Base=LinearInterpolate(A=@n45, B=@n46, Alpha=@n47)) {materialFunction=/Engine/Functions/Engine_MaterialFunctions03/Blends/Blend_ColorDodge}.Result)
 ]
}
